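# Niagara system for the DerangedMile backpack cosmetic, adding a glow and light with green and purple color options.
# Stacked_Plane_Glow bursts five meshes whose spawn group drives increasing scale, decreasing alpha and a material parameter, faking layered glow.
# Decal_Glow_And_Light follows a socket through a socket reader, rendering sprites and lights.
# Both scale brightness by time of day and front-end state, and fade with a shared SystemFade value.
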NiagaraSystem /BRCosmetics/Accessories/FORT_Backpacks/Backpack_M_MED_DerangedMile/FX/NS_Backpack_DerangedMile
dependencies {
 usedDataInterfaces [
  &NiagaraDataInterfaceMeshRendererInfo
  &NiagaraDataInterfaceSocketReader
  &NiagaraDataInterfaceCurve
 ]
 usedDynamicInputs [&FloatFromCurve]
 usedModules [
  &self:SystemSpawnScript.SystemScriptSource.SystemScriptGraph.NiagaraNodeAssignment_0.SetVariables_#1
  &FN_VariantManager
  &SystemState
  &self:SystemSpawnScript.SystemScriptSource.SystemScriptGraph.NiagaraNodeAssignment_2.SetVariables_#2
 ]
}
emitters {
 Decal_Glow_And_Light {
  emitterUpdateScript [
   &EmitterState(
    Inactive Response=::NewEnumerator1
    Life Cycle Mode=::NewEnumerator1
    Loop Behavior=::NewEnumerator1
    Loop Duration Mode=::NewEnumerator1
    UseLoopDelay=false
   )
   &SpawnBurst_Instantaneous(Spawn Count=1, Use Spawn Probability=true) {module=SpawnBurst_Instantaneous001}
   &SpawnBurst_Instantaneous(Spawn Time=0.0667, Use Spawn Probability=false, UseLoopCountLimit=false)
  ]
  particleSpawnScript [
   &InitializeParticle(
    Color Mode=::NewEnumerator0
    Lifetime=2
    Sprite Size Mode=::NewEnumerator3
    Uniform Sprite Size=27.5
    UsePositionOffset=false
    Write Material Random=true
   ) {module=InitializeParticle001}
   /SetVariables(Particles.LightPosition:NiagaraPosition=&AddVectorToPosition(
     Position=$Particles.Position
     Vector=&FN_FrontEndVector3Modifier(Gameplay=<0 -20 0>, Front End=<0 -20 0>)
    )) {isSetParameters=true}
   /SetVariables(Particles.LightSpecularScale=0.225) {isSetParameters=true}
  ]
  particleUpdateScript [
   &ParticleState(Kill Particles When Lifetime Has Elapsed=false)
   &SocketLocation(
    Orientation Sampling=::NewEnumerator1
    SamplePosOffset=true
    Sampled Position Offset=&FN_FrontEndVector3Modifier(Gameplay=<0 0 0>, Front End=<0 0 0>)
    Socket Reader=$System.Socket Reader
    Socket Sampling Mode=::NewEnumerator3
    VelocitySpeedLimit=false
   )
   &ScaleColor(
    Scale RGB=&MakeVectorFromLinearColorRGB(Color=&Multiply_LinearColorByFloat(
      Linear Color=&Multiply_LinearColorByFloat(
       Linear Color=&VariantColorSelector(Variant Amount=0, Original Color=<rgba 0 0.085 0 1>, Variant 1 Color=<rgba 0.0708 0 0.085 1>, Variant 2 Color=?)
       Float=&Multiply_Float(
        A=&FNTimeOfDayScalar(Night Scale=0.8, Morning Scale=0.9, Evening Scale=0.9)
        B=&FN_FrontEndFloatModifier(Gameplay=0.225, Front End=0.4)
       )
      )
      Float=$System.SystemFade
     ))
    ScaleA=true
    ScaleRGB=true
   )
   &FN_Light_Attributes(
    Color=$Particles.Color
    Create Color Attribute=true
    Create Exponent Attribute=true
    Create Radius Attribute=true
    Exponent=3
    Radius=22.5
    Scale Color=50
   )
  ]
  rendererClasses [&NiagaraLightRendererProperties, &NiagaraSpriteRendererProperties]
  renderers [
   {rendererClass=&NiagaraLightRendererProperties, rendererIndex=0}
   {rendererClass=&NiagaraSpriteRendererProperties, rendererIndex=1}
  ]
 }
 Stacked_Plane_Glow {
  emitterUpdateScript [
   &EmitterState(
    Inactive Response=::NewEnumerator1
    Life Cycle Mode=::NewEnumerator1
    Loop Behavior=::NewEnumerator1
    Loop Duration Mode=::NewEnumerator1
    UseLoopDelay=false
   )
   &SpawnBurst_Instantaneous(Spawn Count=5, Use Spawn Probability=false, UseLoopCountLimit=false)
  ]
  particleSpawnScript [
   &FN_SetSpawnGroup(Number of Spawn Groups=5)
   &InitializeParticle(
    Color=&VariantColorSelector(Variant Amount=0, Original Color=<rgba 0 0.1 0 1>, Variant 1 Color=<rgba 0.0958 0 0.115 1>, Variant 2 Color=?)
    Lifetime=5
    Mesh Scale=&Multiply_Vector(
     A=<1 1 1>
     B=&MakeVector(
      X=&Add_Float(A=1, B=&Multiply_FloatByInt(Float=0.1, Integer=&SpawnGroup()))
      Y=&Add_Float(A=1, B=&Multiply_FloatByInt(Float=0.085, Integer=&SpawnGroup()))
      Z=&Add_Float(A=1, B=&Multiply_FloatByInt(Float=0.1, Integer=&SpawnGroup()))
     )
    )
    Mesh Scale Mode=::NewEnumerator4
    Position Offset=<0 2 0>
    UsePositionOffset=true
    Write Material Random=true
   )
   &CameraOffset(Camera Offset Amount=2)
   &DynamicMaterialParameters(
    Index 0 Param 1=&Multiply_FloatByInt(Float=0.1, Integer=&SpawnGroup())
    Module.Index0Float0Write=true
    Module.Index0Float1Write=false
    Module.Index0Float2Write=false
    Module.Index0Float3Write=false
   )
   &InitialMeshOrientation(Mesh Orientation Mode=::NewEnumerator7, Use Rotation Vector=true)
  ]
  particleUpdateScript [
   &ParticleState(Kill Particles When Lifetime Has Elapsed=false, Loop Particles Lifetime=true)
   &ScaleColor(
    Scale Alpha=&Multiply_Float(
     A=&Multiply_Float(A=1, B=$System.SystemFade)
     B=&Power(a=0.7, b=&MakeFloatFromInt(Int=&SpawnGroup()))
    )
    Scale RGB=&VectorFromFloat(Value=&Multiply_Float(
      A=&FNTimeOfDayScalar(Night Scale=0.85, Morning Scale=1, Evening Scale=0.9)
      B=&FN_FrontEndFloatModifier(Gameplay=1, Front End=1.85)
     ))
    ScaleA=true
    ScaleRGB=true
   )
  ]
  rendererClasses [&NiagaraMeshRendererProperties]
  renderers [{rendererClass=&NiagaraMeshRendererProperties, rendererIndex=0}]
 }
}
summary {
 systemName=NS_Backpack_DerangedMile
 emitters [
  {emitterName=Stacked_Plane_Glow, rendererClasses [&NiagaraMeshRendererProperties]}
  {
   emitterName=Decal_Glow_And_Light
   rendererClasses [&NiagaraLightRendererProperties, &NiagaraSpriteRendererProperties]
  }
 ]
 userVariables [
  User.bIsFrontEnd {type=NiagaraBool}
  User.bIsFrontEndPreview {type=NiagaraBool}
  User.Variant {type=NiagaraFloat, defaultValue=0, description=""}
 ]
}
