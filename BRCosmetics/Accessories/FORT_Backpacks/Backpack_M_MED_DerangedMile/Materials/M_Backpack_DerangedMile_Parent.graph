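# Parent material for the Deranged Mile backpack, adding animated panel glow and an inner glass look between the shared 2023 character PreFX and PostFX functions.
# Time-scrolled, repeating U gradients sharpen into pulses that modulate a tiled grayscale pattern, adding tinted emissive within the panel mask.
# A second channel mask blends in inner glass attributes with tinted base color and emissive, alongside an axis-aligned Fresnel term.
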Material /BRCosmetics/Accessories/FORT_Backpacks/Backpack_M_MED_DerangedMile/Materials/M_Backpack_DerangedMile_Parent
graph {
 properties {
  bUseMaterialAttributes=true
  bUsedWithMeshDeformer=true
  bUsedWithSkeletalMesh=true
  pixelDepthOffsetMode=PDOM_Legacy
  shadingModel=MSM_FromMaterialExpression
  usedShadingModels=Default Lit | Subsurface | Clear Coat | Subsurface Profile | Cloth
 }
 nodes [
  n1 = ComponentMask(None=TextureCoordinate()) {r=true}
  n2 = Multiply(A=@n1) {constB=10}
  n3 = Time()
  n4 = Frac(None=Add(A=Frac(None=@n2), B=Multiply(A=@n3) {constB=-0.7}))
  n5 = NamedRerouteUsage()
  LinearInterpolate(
   A=TextureSample() {texture=/BRCosmetics/Accessories/FORT_Backpacks/Backpack_M_MED_DerangedMile/Textures/T_DerangedMile_Backpack_E}.RGB
   B=@n4
   Alpha=@n5
  )
  n6 = MaterialFunctionCall() {materialFunction=&MF_CharacterParent_2023_PreFX}
  MaterialFunctionCall(Material=NamedRerouteUsage(), M=@n6.M, Pre Skinned Local Position=@n6.M, Distance Blend=@n6.M, Fresnel=@n6.M) {materialFunction=&MF_CharacterParent_2023_PostFX}
  NamedRerouteDeclaration(None=@n6.Result) {name=InMaterial, nodeColor=<rgba 0 1 0.0625 1>}
  NamedRerouteDeclaration(None=@n6.Tech Art Mask) {name=InFXMask, nodeColor=<rgba 1 0 0.2578 1>}
  NamedRerouteDeclaration(None=ChannelMaskParameter(None=NamedRerouteUsage()) {maskChannel=Blue, parameterName=PanelPulseMask, defaultValue=<rgba 0 0 1 0>}.R) {name=PanelMask, nodeColor=<rgba 1 0 0.2344 1>}
  n7 = GetMaterialAttributes(None=NamedRerouteUsage()) {attributeGetTypes [#1]}
  NamedRerouteDeclaration(None=@n7.Emissive Color) {name=InEmissive, nodeColor=<rgba 0 1 0.4844 1>}
  n8 = Add(A=TextureCoordinate(), B=ComponentMask(None=Constant3Vector() {constant=<rgba -0.258 0.138 0 0>}) {g=true, r=true})
  n9 = ComponentMask(None=VectorParameter() {defaultValue=<rgba 1 2.7902 0 1>}.RGB) {g=true, r=true}
  n10 = TextureSample(UVs=Multiply(A=Multiply(A=@n8, B=@n9) {constB=12}) {constB=12}) {
   samplerType=SAMPLERTYPE_LinearGrayscale
   texture=&/T_DerangedMile_Pulse
  }
  n11 = Add(
   A=Power(Base=@n10.R, Exp=ScalarParameter() {defaultValue=4, parameterName=Exp})
   B=Step(Y=ScalarParameter() {defaultValue=0.55, parameterName=Y}, X=@n10.R) {constX=0.5, constY=0.5}
  )
  n12 = Cosine(None=Frac(None=Multiply(A=@n1) {constB=10}))
  n13 = SmoothStep(Value=OneMinus(None=ConstantBiasScale(None=@n12))) {constMax=0.5}
  n14 = Power(Base=@n4) {constExponent=32}
  n15 = ScalarParameter() {defaultValue=50, parameterName=B}
  n16 = Multiply(
   A=Multiply(A=Multiply(A=@n11, B=@n13), B=VectorParameter() {parameterName=Param_1, defaultValue=<rgba 10 0 0.1501 1>}.RGB)
   B=Add(A=Power(Base=@n4) {constExponent=3}, B=Multiply(A=@n14, B=@n15))
  )
  NamedRerouteDeclaration(None=Add(A=NamedRerouteUsage(), B=Multiply(A=@n16, B=@n5))) {name=OutEmissive, nodeColor=<rgba 1 0.1641 0 1>}
  n17 = SetMaterialAttributes(MaterialAttributes=@n7.MaterialAttributes, Emissive Color=NamedRerouteUsage()) {attributeSetTypes [#1]}
  n18 = Multiply(A=VectorParameter() {parameterName=Param_2, defaultValue=<rgba 0.0382 0.0356 0.0319 1>}.RGB, B=NamedRerouteUsage())
  n19 = NamedRerouteUsage()
  n20 = Multiply(A=VectorParameter() {parameterName=Param_3, defaultValue=<rgba 0 0.6823 0.0155 1>}.RGB, B=LinearInterpolate(Alpha=@n19) {constA=0.1})
  NamedRerouteDeclaration(None=BlendMaterialAttributes(
    A=@n17
    B=SetMaterialAttributes!(MaterialAttributes=@n17, Base Color=@n18, Emissive Color=@n20) {attributeSetTypes [#2, #1]}
    Alpha=NamedRerouteUsage()
   )) {name=OutMaterial, nodeColor=<rgba 0.6406 0 1 1>}
  NamedRerouteDeclaration(None=ChannelMaskParameter(None=NamedRerouteUsage()) {parameterName=InnerGlassMask}.R) {name=InnerGlassMask, nodeColor=<rgba 0.8281 1 0 1>}
  NamedRerouteDeclaration(None=MaterialFunctionCall(Exponent=ScalarParameter() {defaultValue=2}) {materialFunction=&AxisAlignedFresnel}.Fresnel) {name=Fresnel, nodeColor=<rgba 0 1 0.3672 1>}
  OneMinus(None=@n19)
  TextureSample() {
   samplerType=SAMPLERTYPE_Masks
   texture=/BRCosmetics/Accessories/FORT_Backpacks/Backpack_M_MED_DerangedMile/Textures/T_DerangedMile_Backpack_FX
  }
 ]
}
