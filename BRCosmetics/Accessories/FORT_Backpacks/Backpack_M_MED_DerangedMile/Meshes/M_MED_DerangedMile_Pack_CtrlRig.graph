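# Control Rig for the DerangedMile banana backpack that procedurally animates its peel and teeth bones.
# Chain harmonics waves run first on the four peel chains, then functions layer weighted transform offsets onto the teeth and each peel.
# Each offset weight comes from seeded, timed random values, thresholded, smoothed by AlphaInterp and shaped by a curve, giving irregular twitches.
# A banan_top function for bananas_02 exists but the main graph never calls it.
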ControlRigBlueprint /BRCosmetics/Accessories/FORT_Backpacks/Backpack_M_MED_DerangedMile/Meshes/M_MED_DerangedMile_Pack_CtrlRig
graphs [
 New Function_1_ContainedGraph {nodes [
   RigVMFunctionEntryNode()
   n1 = /Script/RigVM.RigVMFunction_RandomFloat(Duration=3, Maximum=1, Minimum=0, Seed=11) {is=struct RigVMUnitNode, template="Random::Execute(in Seed,in Minimum,in Maximum,in Duration,out Result)"}
   n2 = &RigVMFunction_AlphaInterp(
    InterpSpeedDecreasing=2
    InterpSpeedIncreasing=2
    Scale=1
    bClampResult=true
    bMapRange=true
    Value=RigVMDispatchNode(
     False=1
     True=0
     Condition=/Script/RigVM.RigVMFunction_MathDoubleGreater(B=0.7, A=@n1.Result) {is=struct RigVMUnitNode, template="Greater::Execute(in A,in B,out Result)"}.Result
    ) {template="DISPATCH_RigVMDispatch_If(in Condition,in True,in False,out Result)"}.Result
   )
   n3 = &RigVMFunction_AnimEvalRichCurve(
    Curve=(EditorCurveData=(Keys=((InterpMode=RCIM_Cubic,Time=-0.25),(InterpMode=RCIM_Cubic,Time=0.5,Value=1),(Time=1))))
    SourceMinimum=0
    TargetMinimum=0
    Value=@n2.Result
   )
   > &RigUnit_ModifyTransforms(
    ItemToModify="((Item=(Type=Bone,Name=\"peel_back_01\"),Transform=(Rotation=(X=-0.0957,Y=0.0521,Z=0.005,W=0.994),Translation=(X=0,Y=0,Z=0),Scale3D=(X=1,Y=1,Z=1))),(Item=(Type=Bone,Name=\"peel_back_02\"),Transform=(Rotation=(X=-0.1898,Y=0.1026,Z=0.0199,W=0.9762),Translation=( ...(truncated)"
    WeightMinimum=0
    Weight=@n3.Result
   )
   > RigVMFunctionReturnNode()
  ]}
 New Function_1_ContainedGraph {nodes [
   RigVMFunctionEntryNode()
   n1 = &RigVMFunction_RandomFloat(Duration=4, Maximum=1, Minimum=0, Seed=5) {template="Random::Execute(in Seed,in Minimum,in Maximum,in Duration,out Result)"}
   n2 = &RigVMFunction_AlphaInterp(
    InterpSpeedDecreasing=2
    InterpSpeedIncreasing=2
    Scale=1
    bClampResult=true
    bMapRange=true
    Value=RigVMDispatchNode(
     False=1
     True=0
     Condition=&RigVMFunction_MathDoubleGreater(B=0.8, A=@n1.Result) {template="Greater::Execute(in A,in B,out Result)"}.Result
    ) {template="DISPATCH_RigVMDispatch_If(in Condition,in True,in False,out Result)"}.Result
   )
   n3 = &RigVMFunction_AnimEvalRichCurve(
    Curve=(EditorCurveData=(Keys=((InterpMode=RCIM_Cubic,TangentMode=RCTM_User,Time=-0.25,ArriveTangent=0.0212,LeaveTangent=0.0212),(InterpMode=RCIM_Cubic,Time=0.5,Value=1),(Time=1))))
    SourceMinimum=0
    TargetMinimum=0
    Value=@n2.Result
   )
   > &RigUnit_ModifyTransforms(
    ItemToModify=((Item=(Type=Bone,Name="bananas_02"),Transform=(Rotation=(X=-0.4533,Y=-0.1693,Z=-0.0881,W=0.8707),Translation=(X=0,Y=0,Z=0),Scale3D=(X=1,Y=1,Z=1))))
    WeightMinimum=0
    Weight=@n3.Result
   )
   > RigVMFunctionReturnNode()
  ]}
 New Function_2_ContainedGraph {nodes [
   RigVMFunctionEntryNode()
   n1 = &RigVMFunction_RandomFloat(Duration=5, Maximum=1, Minimum=0, Seed=23) {template="Random::Execute(in Seed,in Minimum,in Maximum,in Duration,out Result)"}
   n2 = &RigVMFunction_AlphaInterp(
    InterpSpeedDecreasing=2
    InterpSpeedIncreasing=2
    Scale=1
    bClampResult=true
    bMapRange=true
    Value=RigVMDispatchNode(
     False=1
     True=0
     Condition=&RigVMFunction_MathDoubleGreater(B=0.7, A=@n1.Result) {template="Greater::Execute(in A,in B,out Result)"}.Result
    ) {template="DISPATCH_RigVMDispatch_If(in Condition,in True,in False,out Result)"}.Result
   )
   n3 = &RigVMFunction_AnimEvalRichCurve(
    Curve=(EditorCurveData=(Keys=((InterpMode=RCIM_Cubic,Time=-0.25),(InterpMode=RCIM_Cubic,Time=0.5,Value=1),(Time=1))))
    SourceMinimum=0
    TargetMinimum=0
    Value=@n2.Result
   )
   > &RigUnit_ModifyTransforms(
    ItemToModify="((Item=(Type=Bone,Name=\"peel_right_01\"),Transform=(Rotation=(X=-0.0087,Y=0.0262,Z=0.0002,W=0.9996),Translation=(X=0,Y=0,Z=0),Scale3D=(X=1,Y=1,Z=1))),(Item=(Type=Bone,Name=\"peel_right_02\"),Transform=(Rotation=(X=-0.0957,Y=0.0521,Z=0.005,W=0.994),Translation ...(truncated)"
    WeightMinimum=0
    Weight=@n3.Result
   )
   > RigVMFunctionReturnNode()
  ]}
 New Function_3_ContainedGraph {nodes [
   RigVMFunctionEntryNode()
   n1 = &RigVMFunction_RandomFloat(Duration=4, Maximum=1, Minimum=0, Seed=17) {template="Random::Execute(in Seed,in Minimum,in Maximum,in Duration,out Result)"}
   n2 = &RigVMFunction_AlphaInterp(
    InterpSpeedDecreasing=2
    InterpSpeedIncreasing=2
    Scale=1
    bClampResult=true
    bMapRange=true
    Value=RigVMDispatchNode(
     False=1
     True=0
     Condition=&RigVMFunction_MathDoubleGreater(B=0.7, A=@n1.Result) {template="Greater::Execute(in A,in B,out Result)"}.Result
    ) {template="DISPATCH_RigVMDispatch_If(in Condition,in True,in False,out Result)"}.Result
   )
   n3 = &RigVMFunction_AnimEvalRichCurve(
    Curve=(EditorCurveData=(Keys=((InterpMode=RCIM_Cubic,Time=-0.25),(InterpMode=RCIM_Cubic,Time=0.5,Value=1),(Time=1))))
    SourceMinimum=0
    TargetMinimum=0
    Value=@n2.Result
   )
   > &RigUnit_ModifyTransforms(
    ItemToModify="((Item=(Type=Bone,Name=\"peel_top_01\"),Transform=(Rotation=(X=0,Y=0.1045,Z=0,W=0.9945),Translation=(X=0,Y=0,Z=0),Scale3D=(X=1,Y=1,Z=1))),(Item=(Type=Bone,Name=\"peel_top_02\"),Transform=(Rotation=(X=-0.1898,Y=0.1026,Z=0.0199,W=0.9762),Translation=(X=0,Y=0,Z=0 ...(truncated)"
    WeightMinimum=0
    Weight=@n3.Result
   )
   > RigVMFunctionReturnNode()
  ]}
 New Function_4_ContainedGraph {nodes [
   RigVMFunctionEntryNode()
   Greater = &RigVMFunction_MathDoubleGreater(
    B=0.5
    A=&RigVMFunction_RandomFloat(Duration=4, Maximum=1, Minimum=0, Seed=22) {template="Random::Execute(in Seed,in Minimum,in Maximum,in Duration,out Result)"}.Result
   ) {template="Greater::Execute(in A,in B,out Result)"}
   AlphaInterp = &RigVMFunction_AlphaInterp(
    InterpSpeedDecreasing=4
    InterpSpeedIncreasing=4
    Scale=1
    bClampResult=true
    bMapRange=true
    Value=RigVMDispatchNode(False=1, True=0, Condition=@Greater.Result) {template="DISPATCH_RigVMDispatch_If(in Condition,in True,in False,out Result)"}.Result
   )
   RigVMFunction_AnimEvalRichCurve = &RigVMFunction_AnimEvalRichCurve(
    Curve=(EditorCurveData=(Keys=((InterpMode=RCIM_Cubic,Time=-0.1,Value=1),(InterpMode=RCIM_Cubic,TangentMode=RCTM_User,Time=0.5,ArriveTangent=-0.0293,LeaveTangent=-0.0293),(Time=1,Value=1))))
    SourceMinimum=0
    TargetMinimum=0
    Value=@AlphaInterp.Result
   )
   > &RigUnit_ModifyTransforms(
    ItemToModify=((Item=(Type=Bone,Name="teeth_back_01"),Transform=(Rotation=(X=0,Y=0.4617,Z=0,W=0.887),Translation=(X=0,Y=0,Z=0),Scale3D=(X=1,Y=1,Z=1))))
    WeightMinimum=0
    Weight=@RigVMFunction_AnimEvalRichCurve.Result
   )
   > &RigUnit_ModifyTransforms(
    ItemToModify=((Item=(Type=Bone,Name="teeth_top_01"),Transform=(Rotation=(X=0,Y=0.4617,Z=0,W=0.887),Translation=(X=0,Y=0,Z=0),Scale3D=(X=1,Y=1,Z=1))))
    WeightMinimum=0
    Weight=@RigVMFunction_AnimEvalRichCurve.Result
   )
   > RigVMFunctionReturnNode()
  ]}
 New Function_ContainedGraph {nodes [
   RigVMFunctionEntryNode()
   > /Script/ControlRig.RigUnit_ChainHarmonicsPerItem(
    ChainRoot=(Type=Bone,Name="peel_left_01")
    DrawWorldOffset=(Rotation=(X=0,Y=0,Z=0,W=1),Translation=(X=0,Y=0,Z=0),Scale3D=(X=1,Y=1,Z=1))
    Pendulum=(bEnabled=False,PendulumStiffness=2,PendulumGravity=(X=0,Y=0,Z=0),PendulumBlend=0.75,PendulumDrag=0.98,PendulumMinimum=0,PendulumMaximum=0.1,PendulumEase=Linear,UnwindAxis=(X=0,Y=1,Z=0),UnwindMinimum=0.2,UnwindMaximum=0.05)
    Reach=(bEnabled=False,ReachTarget=(X=0,Y=0,Z=0),ReachAxis=(X=1,Y=1,Z=1),ReachMinimum=0,ReachMaximum=0,ReachEase=Linear)
    Speed=(X=1,Y=1,Z=1)
    Wave=(bEnabled=True,WaveFrequency=(X=1,Y=0.6,Z=0.8),WaveAmplitude=(X=0.4,Y=0.4,Z=0.4),WaveOffset=(X=0,Y=0,Z=0),WaveNoise=(X=0,Y=0,Z=0),WaveMinimum=0,WaveMaximum=1,WaveEase=Linear)
    WaveCurve=(EditorCurveData=(Keys=((InterpMode=RCIM_Cubic),(InterpMode=RCIM_Cubic,TangentMode=RCTM_User,Time=0.5,Value=1,ArriveTangent=-0.0391,LeaveTangent=-0.0391),(InterpMode=RCIM_Cubic,Time=1))))
    bDrawDebug=True
   ) {is=struct RigVMUnitNode}
   > &RigUnit_ChainHarmonicsPerItem(
    ChainRoot=(Type=Bone,Name="peel_right_01")
    DrawWorldOffset=(Rotation=(X=0,Y=0,Z=0,W=1),Translation=(X=0,Y=0,Z=0),Scale3D=(X=1,Y=1,Z=1))
    Pendulum=(bEnabled=False,PendulumStiffness=2,PendulumGravity=(X=0,Y=0,Z=0),PendulumBlend=0.75,PendulumDrag=0.98,PendulumMinimum=0,PendulumMaximum=0.1,PendulumEase=Linear,UnwindAxis=(X=0,Y=1,Z=0),UnwindMinimum=0.2,UnwindMaximum=0.05)
    Reach=(bEnabled=False,ReachTarget=(X=0,Y=0,Z=0),ReachAxis=(X=1,Y=1,Z=1),ReachMinimum=0,ReachMaximum=0,ReachEase=Linear)
    Speed=(X=1,Y=1,Z=1)
    Wave=(bEnabled=true,WaveFrequency=(X=0,Y=0,Z=0),WaveAmplitude=(X=0.2,Y=0.2,Z=0.2),WaveOffset=(X=0,Y=0,Z=0),WaveNoise=(X=0,Y=0,Z=0),WaveMinimum=0,WaveMaximum=1,WaveEase=Linear)
    WaveCurve=(EditorCurveData=(Keys=((InterpMode=RCIM_Cubic),(InterpMode=RCIM_Cubic,Time=0.5,Value=1),(InterpMode=RCIM_Cubic,Time=1))))
    bDrawDebug=True
   )
   > &RigUnit_ChainHarmonicsPerItem(
    ChainRoot=(Type=Bone,Name="peel_top_01")
    DrawWorldOffset=(Rotation=(X=0,Y=0,Z=0,W=1),Translation=(X=0,Y=0,Z=0),Scale3D=(X=1,Y=1,Z=1))
    Pendulum=(bEnabled=False,PendulumStiffness=2,PendulumGravity=(X=0,Y=0,Z=0),PendulumBlend=0.75,PendulumDrag=0.98,PendulumMinimum=0,PendulumMaximum=0.1,PendulumEase=Linear,UnwindAxis=(X=0,Y=1,Z=0),UnwindMinimum=0.2,UnwindMaximum=0.05)
    Reach=(bEnabled=False,ReachTarget=(X=0,Y=0,Z=0),ReachAxis=(X=1,Y=1,Z=1),ReachMinimum=0,ReachMaximum=0,ReachEase=Linear)
    Speed=(X=1,Y=1,Z=1)
    Wave=(bEnabled=True,WaveFrequency=(X=0,Y=0,Z=0),WaveAmplitude=(X=0.4,Y=0.4,Z=0.4),WaveOffset=(X=0,Y=0,Z=0),WaveNoise=(X=0,Y=0,Z=0),WaveMinimum=0,WaveMaximum=1,WaveEase=Linear)
    WaveCurve=(EditorCurveData=(Keys=((InterpMode=RCIM_Cubic),(InterpMode=RCIM_Cubic,Time=0.5,Value=1),(InterpMode=RCIM_Cubic,Time=1))))
    bDrawDebug=True
   )
   > &RigUnit_ChainHarmonicsPerItem(
    ChainRoot=(Type=Bone,Name="peel_back_01")
    DrawWorldOffset=(Rotation=(X=0,Y=0,Z=0,W=1),Translation=(X=0,Y=0,Z=0),Scale3D=(X=1,Y=1,Z=1))
    Pendulum=(bEnabled=False,PendulumStiffness=2,PendulumGravity=(X=0,Y=0,Z=0),PendulumBlend=0.75,PendulumDrag=0.98,PendulumMinimum=0,PendulumMaximum=0.1,PendulumEase=Linear,UnwindAxis=(X=0,Y=1,Z=0),UnwindMinimum=0.2,UnwindMaximum=0.05)
    Reach=(bEnabled=False,ReachTarget=(X=0,Y=0,Z=0),ReachAxis=(X=1,Y=1,Z=1),ReachMinimum=0,ReachMaximum=0,ReachEase=Linear)
    Speed=(X=1,Y=1,Z=1)
    Wave=(bEnabled=True,WaveFrequency=(X=1,Y=0.6,Z=0.5),WaveAmplitude=(X=0.4,Y=0.4,Z=0.4),WaveOffset=(X=0,Y=0,Z=0),WaveNoise=(X=0,Y=0,Z=0),WaveMinimum=0,WaveMaximum=1,WaveEase=Linear)
    WaveCurve=(EditorCurveData=(Keys=((InterpMode=RCIM_Cubic),(InterpMode=RCIM_Cubic,Time=0.5,Value=1),(InterpMode=RCIM_Cubic,Time=1))))
    bDrawDebug=True
   )
   > RigVMFunctionReturnNode()
  ]}
 New Function_ContainedGraph {nodes [
   RigVMFunctionEntryNode()
   n1 = &RigVMFunction_RandomFloat(Duration=6, Maximum=1, Minimum=0, Seed=5) {template="Random::Execute(in Seed,in Minimum,in Maximum,in Duration,out Result)"}
   n2 = &RigVMFunction_AlphaInterp(
    InterpSpeedDecreasing=2
    InterpSpeedIncreasing=2
    Scale=1
    bClampResult=true
    bMapRange=true
    Value=RigVMDispatchNode(
     False=1
     True=0
     Condition=&RigVMFunction_MathDoubleGreater(B=0.7, A=@n1.Result) {template="Greater::Execute(in A,in B,out Result)"}.Result
    ) {template="DISPATCH_RigVMDispatch_If(in Condition,in True,in False,out Result)"}.Result
   )
   n3 = &RigVMFunction_AnimEvalRichCurve(
    Curve=(EditorCurveData=(Keys=((InterpMode=RCIM_Cubic,Time=-0.25),(InterpMode=RCIM_Cubic,Time=0.5,Value=1),(Time=1))))
    SourceMinimum=0
    TargetMinimum=0
    Value=@n2.Result
   )
   > &RigUnit_ModifyTransforms(
    ItemToModify="((Item=(Type=Bone,Name=\"peel_left_01\"),Transform=(Rotation=(X=0,Y=0.0785,Z=0,W=0.9969),Translation=(X=0,Y=0,Z=0),Scale3D=(X=1,Y=1,Z=1))),(Item=(Type=Bone,Name=\"peel_left_02\"),Transform=(Rotation=(X=0.1873,Y=0.1873,Z=-0.0364,W=0.9636),Translation=(X=0,Y=0,Z ...(truncated)"
    WeightMinimum=0
    Weight=@n3.Result
   )
   > RigVMFunctionReturnNode()
  ]}
 RigVMFunctionLibrary {nodes [
   banan_top = RigVMCollapseNode(ExecuteContext=())
   peel_back = RigVMCollapseNode(ExecuteContext=())
   peel_harmonics = RigVMCollapseNode(ExecuteContext=())
   peel_left = RigVMCollapseNode(ExecuteContext=())
   peel_right = RigVMCollapseNode(ExecuteContext=())
   peel_top = RigVMCollapseNode(ExecuteContext=())
   teeth_back_top = RigVMCollapseNode(ExecuteContext=())
  ]}
 RigVMModel {nodes [
   RigUnit_BeginExecution = &RigUnit_BeginExecution()
   peel_harmonics = RigVMFunctionReferenceNode(ExecuteContext=@RigUnit_BeginExecution.ExecutePin)
   > teeth_back_top = RigVMFunctionReferenceNode()
   > peel_back = RigVMFunctionReferenceNode()
   > peel_right = RigVMFunctionReferenceNode()
   > peel_top = RigVMFunctionReferenceNode()
   > peel_left = RigVMFunctionReferenceNode()
  ]}
]
hierarchy [
 root {loc=<0 0 0>, rot=<0 0 0 1>}
 pack {parent=root, loc=<0 -0.7316 0>, rot=<0 0 0 1>}
 bananas_01 {parent=pack, loc=<0.2422 -9.4878 -3.2906>, rot=<-0.4925 -0.4798 0.5194 -0.5073>}
 teeth_back_01 {parent=bananas_01, loc=<2.5303 0.034 2.236>, rot=<0 -0.0189 0 -0.9998>}
 teeth_top_01 {parent=bananas_01, loc=<2.2244 -0.0954 -3.1345>, rot=<1 0 0 0>}
 peel_top_01 {parent=bananas_01, loc=<0.9386 -2.7808 -2.4893>, rot=<0.7902 -0.1914 -0.3692 0.4501>}
 peel_top_02 {parent=peel_top_01, loc=<2.2668 0 0>, rot=<0 -0.3151 0 -0.949>}
 peel_top_03 {parent=peel_top_02, loc=<2.2668 0 0>, rot=<-0.0242 -0.4759 -0.0644 -0.8768>}
 peel_top_04 {parent=peel_top_03, loc=<2.2668 0 0>, rot=<0 -0.0748 0 -0.9972>}
 peel_top_05 {parent=peel_top_04, loc=<2.2668 0 0>, rot=<0 0 0 1>}
 peel_back_01 {parent=bananas_01, loc=<1.133 3.2572 1.871>, rot=<0.5373 0.2379 -0.2215 -0.7782>}
 peel_back_02 {parent=peel_back_01, loc=<2.2668 0 0>, rot=<0 -0.2686 0 -0.9632>}
 peel_back_03 {parent=peel_back_02, loc=<2.7664 0 0>, rot=<0.0478 -0.429 0.1163 -0.8945>}
 peel_back_04 {parent=peel_back_03, loc=<2.7416 0 0>, rot=<0 0 0 1>}
 peel_right_01 {parent=bananas_01, loc=<0.5955 2.3012 -2.9789>, rot=<0.8505 0.1778 -0.4009 -0.2904>}
 peel_right_02 {parent=peel_right_01, loc=<2.2668 0 0>, rot=<0 -0.2627 0 -0.9649>}
 peel_right_03 {parent=peel_right_02, loc=<2.2668 0 0>, rot=<-0.011 -0.3413 0.0011 -0.9399>}
 peel_right_04 {parent=peel_right_03, loc=<2.2668 0 0>, rot=<0.0548 -0.3189 0.0253 -0.9458>}
 peel_right_05 {parent=peel_right_04, loc=<1.6384 0 0>, rot=<0.0075 -0.6494 -0.039 -0.7594>}
 peel_left_01 {parent=bananas_01, loc=<0.9509 -2.869 2.3297>, rot=<-0.4366 0.2648 0.2027 -0.8355>}
 peel_left_02 {parent=peel_left_01, loc=<2.5412 -0.0268 0.0287>, rot=<-0.0175 -0.1293 -0.0365 -0.9908>}
 peel_left_03 {parent=peel_left_02, loc=<2.4846 0 -0.2346>, rot=<0 -0.3443 0 -0.9388>}
 peel_left_04 {parent=peel_left_03, loc=<2.2628 0.022 -0.2749>, rot=<-0.0055 -0.3339 -0.0702 -0.94>}
 peel_left_05 {parent=peel_left_04, loc=<2.0437 0 0>, rot=<0 0 0 1>}
 bananas_05 {parent=bananas_01, loc=<-0.9045 0 0.0493>, rot=<0 0 -0.0167 -0.9999>}
 bananas_06 {parent=bananas_05, loc=<-3.197 0.1065 0>, rot=<0 0 0 1>}
 bananas_02 {parent=bananas_06, loc=<-4.3121 0.1074 -0.0493>, rot=<-0.0161 -0.0583 -0.0504 -0.9969>}
 bananas_03 {parent=bananas_02, loc=<-5.0233 0 -4.4357>, rot=<0 0.3538 0 -0.9353>}
 curve peel {type=Curve, loc=<0 0 0>, rot=<0 0 0 1>}
]
variables [
 AccumulateLerp_2_Result_X {type=double, default=0}
 NewVar {type=FQuat, default=(X=0,Y=0,Z=0,W=1)}
 Random_3_Duration {type=float, default=0.5}
 speed {type=double, default=0}
 teeth1_float {type=double, default=0}
 teeth_1 {type=bool, default=False}
]
