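# Physics asset for the DerangedMile backpack cosmetic, giving jiggle simulation to its banana bunch and peel bones.
# Capsule bodies cover the pack, two banana bones and four peel chains (back, left, right, top).
# The main banana bone is angularly locked to the pack, and the second banana and each peel root hang off it with limited swing.
# Most links further down the peel chains allow only slight swing, while the back peel tip bends more freely.
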PhysicsAsset /BRCosmetics/Accessories/FORT_Backpacks/Backpack_M_MED_DerangedMile/Meshes/M_MED_DerangedMile_Pack_Physics
bodies {
 bananas_01 [
  {shapeType=Capsule, length=12.3924, radius=1, rotation=<rot 0 90.5008 -81.3741>}
  {shapeType=Capsule, length=15, radius=3, shapeName=bananas_01_capsule, center=<0.0124 7.6299 5.6323>, rotation=<rot 87.5375 0 0>}
  {shapeType=Capsule, length=15, radius=3, shapeName=bananas_01_capsule, center=<-0.0396 -7.6874 6.1146>, rotation=<rot 86.9849 0 0>}
 ]
 bananas_02 [{shapeType=Capsule, length=3.8481, radius=3.1597, shapeName=bananas_02_capsule, center=<-1.6021 0 -2.0224>, rotation=<rot -24.0933 0 -0.0001>}]
 pack [{shapeType=Capsule, length=5, radius=5, center=<0 -3.2268 17.1253>, rotation=<rot 76.0884 90 180>}]
 peel_back_01 [{shapeType=Capsule, length=0, radius=0.7705, shapeName=peel_back_01_capsule}]
 peel_back_02 [{shapeType=Capsule, length=3, radius=0.5, shapeName=peel_back_02_capsule, rotation=<rot -65.6864 0.0005 -0.0001>}]
 peel_back_03 [{shapeType=Capsule, length=2, radius=0.7, shapeName=peel_back_03_capsule, center=<0.817 0 -0.0085>, rotation=<rot -89.4047 -179.9935 179.994>}]
 peel_left_01 [{shapeType=Capsule, length=0, radius=1.1752, shapeName=peel_left_01_capsule}]
 peel_left_02 [{shapeType=Capsule, length=2.5, radius=0.5, shapeName=peel_left_02_capsule, rotation=<rot -65.6864 0 0>}]
 peel_right_01 [{shapeType=Capsule, length=1.908, radius=0.9093, shapeName=peel_right_01_capsule}]
 peel_right_02 [{shapeType=Capsule, length=1.7718, radius=0.9504, shapeName=peel_right_02_capsule}]
 peel_right_03 [{shapeType=Capsule, length=1.3938, radius=0.2335, shapeName=peel_right_03_capsule}]
 peel_top_01 [{shapeType=Capsule, length=2.1681, radius=1.3395, shapeName=peel_top_01_capsule}]
 peel_top_02 [{shapeType=Capsule, length=1.9104, radius=0.7689, shapeName=peel_top_02_capsule}]
 peel_top_03 [{shapeType=Capsule, length=1.6278, radius=1.2743, shapeName=peel_top_03_capsule}]
}
constraints [
 {bone1Name=bananas_01, bone2Name=pack, swing1LimitDegrees=0, swing2LimitDegrees=0, twistLimitDegrees=0}
 {bone1Name=peel_back_01, bone2Name=bananas_01, swing1LimitDegrees=1, swing2LimitDegrees=20, twistLimitDegrees=0}
 {bone1Name=peel_back_02, bone2Name=peel_back_01, swing1LimitDegrees=1, swing2LimitDegrees=1, twistLimitDegrees=0}
 {bone1Name=peel_left_01, bone2Name=bananas_01, swing1LimitDegrees=1, swing2LimitDegrees=20, twistLimitDegrees=0}
 {bone1Name=peel_left_02, bone2Name=peel_left_01, swing1LimitDegrees=1, swing2LimitDegrees=1, twistLimitDegrees=0}
 {bone1Name=peel_right_01, bone2Name=bananas_01, swing1LimitDegrees=1, swing2LimitDegrees=20, twistLimitDegrees=0}
 {bone1Name=peel_right_02, bone2Name=peel_right_01, swing1LimitDegrees=1, swing2LimitDegrees=1, twistLimitDegrees=0}
 {bone1Name=peel_right_03, bone2Name=peel_right_02, swing1LimitDegrees=1, swing2LimitDegrees=1, twistLimitDegrees=0}
 {bone1Name=peel_top_01, bone2Name=bananas_01, swing1LimitDegrees=1, swing2LimitDegrees=20, twistLimitDegrees=0}
 {bone1Name=peel_top_02, bone2Name=peel_top_01, swing1LimitDegrees=1, swing2LimitDegrees=1, twistLimitDegrees=0}
 {bone1Name=peel_top_03, bone2Name=peel_top_02, swing1LimitDegrees=1, swing2LimitDegrees=1, twistLimitDegrees=0}
 {bone1Name=bananas_02, bone2Name=bananas_01, swing1LimitDegrees=1, swing2LimitDegrees=20, twistLimitDegrees=1}
 {bone1Name=peel_back_03, bone2Name=peel_back_02, swing1LimitDegrees=1, swing2LimitDegrees=30, twistLimitDegrees=1}
]
