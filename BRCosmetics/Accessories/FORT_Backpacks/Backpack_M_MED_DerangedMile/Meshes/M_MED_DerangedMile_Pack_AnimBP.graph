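# Animation Blueprint for the Deranged Mile backpack cosmetic, adding procedural motion to its bananas_01 and pack bones.
# A base pose runs through a Control Rig and a time-based sine offset into a cached pose.
# Front end adds an idle animation; otherwise two AnimDynamics nodes and a RigidBody, weighted by the enable_dynamics curve, feed a pack offset.
# The event graph derives that offset from spine twist, behind a skydiving check.
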AnimBlueprint /BRCosmetics/Accessories/FORT_Backpacks/Backpack_M_MED_DerangedMile/Meshes/M_MED_DerangedMile_Pack_AnimBP
target_skeleton=&/M_MED_DerangedMile_Pack_Skeleton
graphs [
 AnimGraph {nodes [
   n1 = BreakTransform(InTransform=Get player'smotion() -> (player'smotion:Transform))
   n2 = "Transform (Modify) Bone - Bone: bananas_01"!(ComponentPose, Alpha=0.1, Translation=@n1, Rotation=@n1.Rotation) -> (Pose) {rotationMode=BMM_Additive, translationMode=BMM_Additive, alphaBoolBlend {blendOption=Linear}, boneToModify {boneName=bananas_01}}
   "Transform (Modify) Bone - Bone: bananas_01"(ComponentPose=@n2, Alpha=1, Translation="0,2,-12", Rotation="-10,0,0") {
    rotationMode=BMM_Additive
    translationMode=BMM_Additive
    alphaBoolBlend {blendOption=Linear}
    boneToModify {boneName=bananas_01}
    rotation {pitch=-10}
    translation {y=2, z=-12}
   }
   n3 = "Sequence Player (additive) 'Backpack_DerangedMile_Idle_Anim_WEP'"() {sequence=/BRCosmetics/Accessories/FORT_Backpacks/Backpack_M_MED_DerangedMile/Animations/Backpack_DerangedMile_Idle_Anim_WEP}
   n4 = Use cached pose 'Dynamics'()
   n5 = Anim Dynamics - Bone: bananas_01!(ComponentPose=Local To Component(LocalPose=Use cached pose 'Dynamics'()), AlphaCurveName:Name=enable_dynamics, ExternalForce="0,0,0") -> (Pose) {
    alphaInputType=Curve
    angularSpringConstant=350
    bLinearSpring=true
    bOverrideLinearDamping=true
    bUseGravityOverride=true
    linearDampingOverride=0.5
    linearSpringConstant=100
    simulationSpace=RootRelative
    alphaScaleBiasClamp {bClampResult=true, clampMax=0.45}
    boundBone {boneName=bananas_01}
    componentAppliedLinearAccClamp=<3000 3000 5000>
    componentLinearAccScale {z=0.4}
    componentLinearVelScale {z=0.3}
    gravityOverride {z=-1100}
    physicsBodyDefinitions [{
      boundBone {boneName=bananas_01}
      boxExtents=<100 100 100>
      constraintSetup {angularTargetAxis=AxisZ, twistAxis=AxisZ, angularLimitsMax=<0 0 0>, angularLimitsMin=<0 0 0>, angularTarget=<0 0 0>, linearAxesMax=<1 0 0>}
      localJointOffset=<0.05 0 0>
     }]
   }
   n6 = Anim Dynamics - Bone: bananas_01(ComponentPose=@n5, AlphaCurveName=enable_dynamics, ExternalForce="0,0,0") {
    alphaInputType=Curve
    bLinearSpring=true
    bOverrideLinearDamping=true
    bUseGravityOverride=true
    linearDampingOverride=0.8
    linearSpringConstant=50
    simulationSpace=RootRelative
    alphaScaleBiasClamp {bClampResult=true, clampMax=0.5}
    boundBone {boneName=bananas_01}
    componentAppliedLinearAccClamp=<3000 3000 5000>
    componentLinearAccScale=<0.15 0.15>
    componentLinearVelScale=<0.005 0.005>
    gravityOverride {z=-980}
    physicsBodyDefinitions [{
      boundBone {boneName=bananas_01}
      boxExtents=<10 10 10>
      constraintSetup {
       angularTargetAxis=AxisZ
       twistAxis=AxisZ
       angularLimitsMax=<0 10 0>
       angularLimitsMin=<0 0 0>
       angularTarget=<0 0 0>
       linearAxesMin=<0 -1 -1>
      }
      localJointOffset=<0.3 0 0>
     }]
   }
   n7 = RigidBody!(ComponentPose=@n6, AlphaCurveName:Name=enable_dynamics, ExternalForce) -> (Pose) {
    alphaInputType=Curve
    bOverrideWorldGravity=true
    overridePhysicsAsset=&/M_MED_DerangedMile_Pack_Physics
    simulationSpace=BaseBoneSpace
    alphaScaleBiasClamp {bClampResult=true, clampMax=0.6}
    baseBoneRef {boneName=root}
    overrideWorldGravity {z=-500}
    simSpaceSettings {maxAngularAcceleration=0, maxAngularVelocity=0, maxLinearAcceleration=500, maxLinearVelocity=500, worldAlpha=1}
   }
   n8 = "Transform (Modify) Bone - Bone: pack"(ComponentPose=@n7, Translation=Get backpack_translation(), Rotation=Get backpack_rotate()) {rotationMode=BMM_Additive, translationMode=BMM_Additive, alphaBoolBlend {blendOption=Linear}, boneToModify {boneName=pack}}
   Output Pose(Result=Blend Poses by bool(BlendPose_0=Apply Additive(Base=@n4, Additive=@n3), BlendPose_1=Component To Local(ComponentPose=@n8), bActiveValue=Get bIsInFrontEnd())) {layerGroup=None}
   n9 = Control Rig!(
    Source=Sequence Player 'Backpack_DerangedMile_Base_Pose_WEP'() {sequence=/BRCosmetics/Accessories/FORT_Backpacks/Backpack_M_MED_DerangedMile/Animations/Backpack_DerangedMile_Base_Pose_WEP}
    AlphaCurveName:Name=enable_dynamics
   ) -> (Pose) {
    alphaInputType=Curve
    controlRigAssetReference {blueprintRigClass=&/M_MED_DerangedMile_Pack_CtrlRig.M_MED_DerangedMile_Pack_CtrlRig_C}
   }
   n10 = "Transform (Modify) Bone - Bone: bananas_01"!(ComponentPose=Local To Component(LocalPose=@n9), Alpha=0.8, Translation=Get body move() -> (body move:Vector)) -> (Pose) {translationMode=BMM_Additive, translationSpace=BCS_BoneSpace, alphaBoolBlend {blendOption=Linear}, boneToModify {boneName=bananas_01}}
   Save cached pose 'Dynamics'(Pose=Component To Local(ComponentPose=@n10)) {cacheName=Dynamics}
  ]}
 EventGraph {nodes [
   Event BlueprintUpdateAnimation()
   > n1 = Sequence!(execute) -> (then_0, then_1, then_2:Exec)
   > Set curve_alpha(curve_alpha:float=MapRangeClamped(Value=GetCurveValue(CurveName=disable dynamics), InRangeA=0, InRangeB=1, OutRangeA=1, OutRangeB=0)) -> (Output_Get:float)
   n2 = GetRealTimeSeconds()
   n3 = "Sin (Radians)"(A=float + float(A=float * float(A=@n2, B=0.45) {is=title K2Node_PromotableOperator}))
   > Set body move(body move:Vector=MakeVector(X=float * float(A=@n3, B=1.1), Y=0)) -> (Output_Get:Vector)
   Set player'smotion(
    execute=@n1.then_2
    player'smotion:Transform=GetBoneTransform!(
     self:Skinned Mesh Component Object Reference=GetOwnerSkeletalMeshComponent()
     InBoneName:Name=spine_05
     TransformSpace:ERelativeTransformSpace Enum=RTS_Component
    ) -> (ReturnValue:Transform)
   ) -> (Output_Get:Transform)
   Branch(execute=@n1.then_1, Condition=Get bIsSkyDiving())
   n4 = GetOwnerSkeletalMeshComponent()
   n5 = BreakRotator(InRot=BreakTransform(InTransform=GetDeltaTransformFromRefPose(self=@n4, BoneName=spine_05, BaseName=spine_01)).Rotation)
   n6 = MapRangeClamped(Value=@n5.Yaw, InRangeA=-10, InRangeB=-50, OutRangeB=40)
   > Set backpack_rotate(backpack_rotate=MakeRotator(Roll=@n6, Pitch=float / float(A=@n6, B=5)))
   > Set backpack_translation(backpack_translation=MakeVector(Y=MapRangeClamped(Value=@n5.Yaw, InRangeA=-15, InRangeB=-50, OutRangeB=-16)))
   TryGetPawnOwner()
   Reroute Node(InputPin:Wildcard) -> (OutputPin:Wildcard)
   Reroute Node()
  ]}
 NewEventDispatcher {nodes [NewEventDispatcher() {is=title K2Node_FunctionEntry}]}
]
variables [
 backpack_rotate {type=Rotator}
 backpack_translation {type=Vector}
 BackpackRotate {type=number}
 body move {type=Vector}
 curve_alpha {type=number}
 NewEventDispatcher {default=()}
 peel_back_Animation_Enabled {type=boolean}
 peel_top_Animation_Enabled {type=boolean}
 player'smotion {type=Transform, default {location=<0 0 0>, rotation=<rot 0 0 0>, scale=<1 1 1>}}
]
